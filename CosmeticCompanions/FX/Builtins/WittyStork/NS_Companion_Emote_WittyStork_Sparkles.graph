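# Niagara sparkle effect for the WittyStork companion's emote in CosmeticCompanions.
# The active sprite emitter spawns short-lived, SubUV-animated sparkles on the surface band of a rotated half-cylinder, growing them in by curve.
# User.SparkleSpawnLength lerps spawn rate, cylinder cross-section and vertical offset together, while User.Scale sizes shape and sprites.
# A disabled burst mesh emitter passes SparkleSpawnLength to a dynamic material parameter and fades alpha late in life.
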NiagaraSystem /CosmeticCompanions/FX/Builtins/WittyStork/NS_Companion_Emote_WittyStork_Sparkles
dependencies {
 usedDataInterfaces [&NiagaraDataInterfaceCurve, &NiagaraDataInterfaceSpriteRendererInfo]
 usedDynamicInputs []
 usedModules [&SystemState]
}
emitters {
 Minimal {
  emitterUpdateScript [
   &EmitterState()
   &SpawnRate(SpawnRate=&Lerp_Float(A=5, B=13, Alpha=$User.SparkleSpawnLength), Use Spawn Probability=false)
  ]
  particleSpawnScript [
   &InitializeParticle(
    Color=<rgba 1 1 1 0.5>
    Lifetime Max=0.8
    Lifetime Min=0.4
    Lifetime Mode=::NewEnumerator1
    Sprite Rotation Mode=::NewEnumerator2
    Sprite Size Mode=::NewEnumerator1
    Uniform Sprite Size Max=25
    Uniform Sprite Size Min=11
    UsePositionOffset=false
    Write Material Random=true
   )
   &ShapeLocation(
    Cylinder Height=&Multiply_Float(A=7, B=$User.Scale)
    Cylinder Radius=&Multiply_Float(A=40, B=$User.Scale)
    Hemicircle Internal Cap=false
    Hemicircle Y=true
    MaskSpawnGroup=false
    Non Uniform Scale=&MakeVector(
     X=&Lerp_Float(A=0, B=0.9, Alpha=$User.SparkleSpawnLength)
     Y=&Lerp_Float(A=0, B=0.9, Alpha=$User.SparkleSpawnLength)
     Z=1
    )
    Offset=&MakeVector(
     X=0
     Y=0
     Z=&Multiply_Float(A=&Lerp_Float(A=50, B=22, Alpha=$User.SparkleSpawnLength), B=$User.Scale)
    )
    Offset Mode=::NewEnumerator0
    Override Seed=false
    Rotation Angle=270
    Rotation Axis=<0 -1 0>
    Rotation Mode=::NewEnumerator1
    Shape Primitive=::NewEnumerator1
    Surface Only=true
    Surface Only Band Thickness=0.19
    Use Endcaps In Surface Only Mode=false
    UseOwnerScale=true
   )
  ]
  particleUpdateScript [
   &ParticleState()
   &ScaleSpriteSize(
    Uniform Curve Scale=$User.Scale
    Uniform Curve Sprite Scale=DataInterface() {propertyValues {Curve {keys [{tangentMode=RCTM_User, time=0, value=0}, {tangentMode=RCTM_User, time=0.5, value=1}]}}}
   )
   &SubUVAnimation(
    "Sprite Renderer "=DataInterface() {propertyValues {SpriteRenderer=&self:Minimal.NiagaraSpriteRendererProperties_0}}
    UseEndFrame=false
    UseStartFrame=false
   )
  ]
  rendererClasses [&NiagaraSpriteRendererProperties]
  renderers [{rendererClass=&NiagaraSpriteRendererProperties}]
 }
 Minimal001 {
  bEnabled=false
  emitterUpdateScript [
   &EmitterState()
   &SpawnBurst_Instantaneous(Use Spawn Probability=false, UseLoopCountLimit=false)
  ]
  particleSpawnScript [&InitializeParticle(
    Color=<rgba 1 1 1 0.8>
    Lifetime=3
    Mesh Scale Mode=::NewEnumerator3
    Mesh Uniform Scale=&Multiply_Float(A=1, B=$User.Scale)
    UsePositionOffset=false
    Write Material Random=true
   )]
  particleUpdateScript [
   &ParticleState()
   &DynamicMaterialParameters(
    Index 0 Param 1=$User.SparkleSpawnLength
    Module.Index0Float0Write=true
    Module.Index0Float1Write=true
    Module.Index0Float2Write=true
    Module.Index0Float3Write=true
   )
   &ScaleColor(
    Scale Alpha=&FloatFromCurve(FloatCurve=DataInterface() {propertyValues {Curve {keys [<at 0.8 1>,<at 1 0>]}}})
    ScaleA=true
    ScaleRGB=true
   )
  ]
  rendererClasses [&NiagaraMeshRendererProperties]
  renderers [{rendererClass=&NiagaraMeshRendererProperties, rendererIndex=0}]
 }
}
summary {
 systemName=NS_Companion_Emote_WittyStork_Sparkles
 emitters [
  {emitterName=Minimal, rendererClasses [&NiagaraSpriteRendererProperties]}
  {bEnabled=false, emitterName=Minimal001, rendererClasses [&NiagaraMeshRendererProperties]}
 ]
 userVariables [
  User.Scale {type=NiagaraFloat, defaultValue=0.8, description=""}
  User.SparkleSpawnLength {type=NiagaraFloat, defaultValue=1, description=""}
 ]
}
