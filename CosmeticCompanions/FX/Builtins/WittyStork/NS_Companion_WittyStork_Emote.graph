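# Niagara system for the WittyStork companion's emote, emitting short-lived, pale translucent sparkle sprites from a thin surface band of an offset shape.
# User.SparkleSpawnLength interpolates spawn rate, shape width and height offset, so one float scales the whole effect.
# Sprites grow via a size curve while a stepped curve writes indices 0 to 9 into a dynamic material parameter, likely flipbook frames.
# A system-age vertical motion step and a mesh burst emitter are disabled.
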NiagaraSystem /CosmeticCompanions/FX/Builtins/WittyStork/NS_Companion_WittyStork_Emote
dependencies {usedDataInterfaces [&NiagaraDataInterfaceCurve], usedDynamicInputs [], usedModules [&SystemState]}
emitters {
 Minimal {
  emitterUpdateScript [
   &EmitterState()
   &SpawnRate(SpawnRate=&Lerp_Float(A=10, B=25, Alpha=$User.SparkleSpawnLength), Use Spawn Probability=false)
  ]
  particleSpawnScript [
   &InitializeParticle(
    Color=<rgba 0.8132 0.8166 1 0.5>
    Lifetime=0.5
    Sprite Rotation Mode=::NewEnumerator2
    Sprite Size Mode=::NewEnumerator1
    Uniform Sprite Size Max=16
    Uniform Sprite Size Min=7
    UsePositionOffset=false
    Write Material Random=true
   )
   &ShapeLocation(
    Cylinder Height=7
    Cylinder Radius=45
    Hemicircle Internal Cap=false
    Hemicircle Y=true
    MaskSpawnGroup=false
    Non Uniform Scale=&MakeVector(
     X=&Lerp_Float(A=0, B=0.75, Alpha=$User.SparkleSpawnLength)
     Y=&Lerp_Float(A=0, B=0.8, Alpha=$User.SparkleSpawnLength)
     Z=1
    )
    Offset=&MakeVector(X=0, Y=0, Z=&Lerp_Float(A=28, B=32, Alpha=$User.SparkleSpawnLength))
    Offset Mode=::NewEnumerator0
    Override Seed=false
    Rotation Angle=270
    Rotation Axis=<0 -1 0>
    Rotation Mode=::NewEnumerator1
    Shape Primitive=::NewEnumerator1
    Surface Only=true
    Surface Only Band Thickness=0.2
    Use Endcaps In Surface Only Mode=false
    UseOwnerScale=true
   )
  ]
  particleUpdateScript [
   &ParticleState()
   &ScaleSpriteSize(Uniform Curve Sprite Scale=DataInterface() {propertyValues {Curve {keys [{tangentMode=RCTM_User, time=0, value=0}, {tangentMode=RCTM_User, time=0.5, value=1}]}}})
   &DynamicMaterialParameters(
    Index 0 Param 1=1
    Index 0 Param 2=0.05
    Index 0 Param 3=&FloatFromCurve(FloatCurve=DataInterface() {propertyValues {Curve {
        defaultValue=1
        keys [<at 0 0 RCIM_Constant>,<at 0.1111 1 RCIM_Constant>,<at 0.2222 2 RCIM_Constant>,<at 0.3333 3 RCIM_Constant>,<at 0.4444 4 RCIM_Constant>,<at 0.5556 5 RCIM_Constant>,<at 0.6667 6 RCIM_Constant>,<at 0.7778 7 RCIM_Constant>,<at 0.8888 8 RCIM_Constant>,<at 1 9 RCIM_Constant>]
       }}})
    Module.Index0Float0Write=true
    Module.Index0Float1Write=true
    Module.Index0Float2Write=true
    Module.Index0Float3Write=true
   )
   /SetVariables(Particles.Position=&AddVectorToPosition(
     Position=$Particles.Initial.Position
     Vector=&MakeVector(
      X=0
      Y=0
      Z=&FloatFromCurve(
       FloatCurve=DataInterface() {propertyValues {Curve {keys [
           {arriveTangent=31.1806, interpMode=RCIM_Cubic, leaveTangent=31.1806, tangentMode=RCTM_User, time=0.2, value=-20}
           {arriveTangent=11.0679, interpMode=RCIM_Cubic, leaveTangent=11.0679, tangentMode=RCTM_User, time=0.912, value=15}
          ]}}}
       CurveIndex=$Engine.System.Age
      )
     )
    )) {enabled=false, isSetParameters=true}
  ]
  rendererClasses [&NiagaraSpriteRendererProperties]
  renderers [{rendererClass=&NiagaraSpriteRendererProperties}]
 }
 Minimal001 {
  bEnabled=false
  emitterUpdateScript [
   &EmitterState()
   &SpawnBurst_Instantaneous(Use Spawn Probability=false, UseLoopCountLimit=false)
  ]
  particleSpawnScript [&InitializeParticle(Color Mode=::NewEnumerator0, Lifetime=5, UsePositionOffset=false, Write Material Random=true)]
  particleUpdateScript [
   &ParticleState()
   &ScaleSpriteSize()
  ]
  rendererClasses [&NiagaraMeshRendererProperties]
  renderers [{rendererClass=&NiagaraMeshRendererProperties, rendererIndex=0}]
 }
}
summary {
 systemName=NS_Companion_WittyStork_Emote
 emitters [
  {emitterName=Minimal, rendererClasses [&NiagaraSpriteRendererProperties]}
  {bEnabled=false, emitterName=Minimal001, rendererClasses [&NiagaraMeshRendererProperties]}
 ]
 userVariables [User.SparkleSpawnLength {type=NiagaraFloat, defaultValue=1, description=""}]
}
